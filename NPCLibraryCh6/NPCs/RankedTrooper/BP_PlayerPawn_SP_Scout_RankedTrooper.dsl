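; Pawn Blueprint for the Scout Ranked Trooper, with helper functions for marking nearby actors.
; GetActorsToMark sphere-overlaps three object types around the pawn, using the MarkRadius scalable float.
; ModifyFilterData rebuilds each indicator filter entry, applying MarkDuration to indicated and stenciled data, overriding several flags and substituting the supplied radius.
; BeginPlay and construction only call the parent, and neither helper is called in this graph.
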(event EventBeginPlay
  (|CallParentFunction))

(event Collision|EventActorBeginOverlap
  (params (OtherActor :type Actor%20Object%20Reference))
  (|CallParentFunction OtherActor) ; This node is disabled and will not be called.\nDrag off pins to build functionality.
)

(event EventTick
  (params (DeltaSeconds :type Float%20%28single-precision%29))
  (|CallParentFunction DeltaSeconds) ; This node is disabled and will not be called.\nDrag off pins to build functionality.
)

(fn GetActorsToMark
  (bind _outactors (Collision|SphereOverlapActors (Transformation|GetActorLocation) (ScalableFloat|GetValueatLevel (Variables|Default|GetMarkRadius)) (Utilities|Array|MakeArray "ObjectTypeQuery2" "ObjectTypeQuery3" "ObjectTypeQuery1")))
  (return _outactors))

(fn ModifyFilterData
  (params (Array :type Array%20of%20Indicated%20Actor%20Data%20With%20Filter%20Structures) (Radius :type Float%20%28double-precision%29))
  (for _array_element Array
    (Utilities|Array|Add (LocalVariables|GetOutDatas) (Utilities|Struct|MakeIndicatedActorDatawithFilter (Utilities|Struct|BreakIndicatedActorDatawithFilter _array_element) (.IndicatedActorTagQuery (Utilities|Struct|BreakIndicatedActorDatawithFilter _array_element)) (.ObjectTypes (Utilities|Struct|BreakIndicatedActorDatawithFilter _array_element)) (.ActorClassFilter (Utilities|Struct|BreakIndicatedActorDatawithFilter _array_element)) (.WithAffiliation (Utilities|Struct|BreakIndicatedActorDatawithFilter _array_element)) (Utilities|Struct|MakeIndicatedActorData (Utilities|Struct|BreakIndicatedActorData (.IndicatedData (Utilities|Struct|BreakIndicatedActorDatawithFilter _array_element))) (ScalableFloat|GetValueatLevel (Variables|Default|GetMarkDuration)) (.StepTime (Utilities|Struct|BreakIndicatedActorData (.IndicatedData (Utilities|Struct|BreakIndicatedActorDatawithFilter _array_element)))) (.IndicatorOffset (Utilities|Struct|BreakIndicatedActorData (.IndicatedData (Utilities|Struct|BreakIndicatedActorDatawithFilter _array_element)))) (.IndicatorDBNOOffset (Utilities|Struct|BreakIndicatedActorData (.IndicatedData (Utilities|Struct|BreakIndicatedActorDatawithFilter _array_element)))) (.ShareActorWith (Utilities|Struct|BreakIndicatedActorData (.IndicatedData (Utilities|Struct|BreakIndicatedActorDatawithFilter _array_element)))) (.ShareActorWithMask (Utilities|Struct|BreakIndicatedActorData (.IndicatedData (Utilities|Struct|BreakIndicatedActorDatawithFilter _array_element)))) true (.DisplayTeamOverride (Utilities|Struct|BreakIndicatedActorData (.IndicatedData (Utilities|Struct|BreakIndicatedActorDatawithFilter _array_element)))) (.ScaleAndOpacityData (Utilities|Struct|BreakIndicatedActorData (.IndicatedData (Utilities|Struct|BreakIndicatedActorDatawithFilter _array_element)))) (.ParticleSystemData (Utilities|Struct|BreakIndicatedActorData (.IndicatedData (Utilities|Struct|BreakIndicatedActorDatawithFilter _array_element)))) (.Sound (Utilities|Struct|BreakIndicatedActorData (.IndicatedData (Utilities|Struct|BreakIndicatedActorDatawithFilter _array_element)))) (.StateImageOverride (Utilities|Struct|BreakIndicatedActorData (.IndicatedData (Utilities|Struct|BreakIndicatedActorDatawithFilter _array_element)))) -1 (.bClampToScreen (Utilities|Struct|BreakIndicatedActorData (.IndicatedData (Utilities|Struct|BreakIndicatedActorDatawithFilter _array_element)))) false false false (.OnActorAdded (Utilities|Struct|BreakIndicatedActorData (.IndicatedData (Utilities|Struct|BreakIndicatedActorDatawithFilter _array_element)))) (.OnActorRemoved (Utilities|Struct|BreakIndicatedActorData (.IndicatedData (Utilities|Struct|BreakIndicatedActorDatawithFilter _array_element))))) (Utilities|Struct|MakeStenciledActorData (Utilities|Struct|BreakStenciledActorData (.StenciledData (Utilities|Struct|BreakIndicatedActorDatawithFilter _array_element))) (ScalableFloat|GetValueatLevel (Variables|Default|GetMarkDuration)) (.StepTime (Utilities|Struct|BreakStenciledActorData (.StenciledData (Utilities|Struct|BreakIndicatedActorDatawithFilter _array_element)))) (.ShareActorWithMask (Utilities|Struct|BreakStenciledActorData (.StenciledData (Utilities|Struct|BreakIndicatedActorDatawithFilter _array_element)))) true (.DisplayTeamOverride (Utilities|Struct|BreakStenciledActorData (.StenciledData (Utilities|Struct|BreakIndicatedActorDatawithFilter _array_element)))) (.ParticleSystemData (Utilities|Struct|BreakStenciledActorData (.StenciledData (Utilities|Struct|BreakIndicatedActorDatawithFilter _array_element)))) (.Sound (Utilities|Struct|BreakStenciledActorData (.StenciledData (Utilities|Struct|BreakIndicatedActorDatawithFilter _array_element)))) (.FriendlyStencilName (Utilities|Struct|BreakStenciledActorData (.StenciledData (Utilities|Struct|BreakIndicatedActorDatawithFilter _array_element)))) (.EnemyStencilName (Utilities|Struct|BreakStenciledActorData (.StenciledData (Utilities|Struct|BreakIndicatedActorDatawithFilter _array_element))))) (Variables|GetRadius) (.OverlapHalfHeight (Utilities|Struct|BreakIndicatedActorDatawithFilter _array_element)))))
  (return (LocalVariables|GetOutDatas)))

(fn ConstructionScript
  (|CallParentFunction))
